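; Base Blueprint for a state-driven vending machine that switches its screen look and plays player-approach audio.
; Reaction side effects look up their tag in ReactionTaxModifiers and pass the inverted value to material parameters, with game time, to transition between regular and Baron screens.
; Dialog marker changes from the local player store the interaction state and trigger approach sounds in InteractionRange or Conversation.
; The leave-sound branch is unreachable behind a constant true condition.
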(event EventBeginPlay
  (|CallParentFunction)
  (EventDispatchers|BindEventtoOnDialogMarkerInteractionStateChanged (Variables|Components|GetNonPlayerConversationComponent) (delegate Custom|OnInteractionStateChanged)))

(event Custom|OnInteractionStateChanged
  (params (CurrentInteractionState :type EDialogMarkerInteractionState%20Enum :default "Conversation") (InteractingController :type Fort%20Player%20Controller%20Object%20Reference))
  (bind _markerinteractionstate (Variables|Interaction|GetMarkerInteractionState))
  (bind _returnvalue (Utilities|Enum|Equal(Enum) _markerinteractionstate "InteractionRange"))
  (bind _returnvalue_2 (Utilities|Enum|Equal(Enum) _markerinteractionstate "Conversation"))
  (bind _returnvalue_3 (or _returnvalue_2 _returnvalue))
  (if (Pawn|IsLocalPlayerController InteractingController)
    (if (Math|Boolean|NotEqual(Boolean) _returnvalue_3 (or (Utilities|Enum|Equal(Enum) CurrentInteractionState "Conversation") (Utilities|Enum|Equal(Enum) CurrentInteractionState "InteractionRange")))
      (Variables|Interaction|SetMarkerInteractionState CurrentInteractionState)
      (else
        (Variables|Interaction|SetMarkerInteractionState CurrentInteractionState)))
    (if _returnvalue_3
      (CallFunction|OnPlayerApproach))))

(event Collision|EventActorBeginOverlap
  (params (OtherActor :type Actor%20Object%20Reference))
  (|CallParentFunction OtherActor) ; This node is disabled and will not be called.\nDrag off pins to build functionality.
)

(event EventTick
  (params (DeltaSeconds :type Float%20%28single-precision%29))
  (|CallParentFunction DeltaSeconds) ; This node is disabled and will not be called.\nDrag off pins to build functionality.
)

(event Custom|OnScreenStateChanged
  (params (floatMultiplier :type Float%20%28double-precision%29))
  (bind _staticmeshcomponent (Variables|Debug|GetStaticMeshComponent))
  ; Triggers visual Transition between Regular VM and Baron Screen
  (bind _output_get (Variables|StateDrivenActorSettings|SetBaronSwitch floatMultiplier))
  (Rendering|Material|SetScalarParameterValueonMaterials _staticmeshcomponent "BP_BaronMaterial_ON/OFF" _output_get)
  (Rendering|Material|SetScalarParameterValueonMaterials _staticmeshcomponent "BP_GetTime" (Utilities|Time|GetGameTimeinSeconds)))

(event SideEffects|EventBP_ApplySideEffect
  (params (ReactionSource :type Object%20Reference) (TargetObject :type Object%20Reference) (SideEffectTriggerTag :type Gameplay%20Tag%20Structure%20%28by%20ref%29) (SideEffectTriggerType :type EBlueprintableSideEffectTriggerType%20Enum%20%28by%20ref%29 :default "TriggeredAsTarget"))
  (bind _reactiontaxmodifiers (Variables|StateDrivenActorSettings|GetReactionTaxModifiers))
  (bind (_value _returnvalue) (Utilities|Map|Find _reactiontaxmodifiers SideEffectTriggerTag))
  (if _returnvalue
    (CallFunction|OnScreenStateChanged (Math|Conversions|ToFloat(Boolean) (not (ScalableFloat|GetValueasBool _value))))))

(event Custom|OnPlayerApproach
  (CallFunction|PlayApproachSounds))

(event Custom|PlayApproachSounds
  (bind _onplayerapproachscreensound (Variables|StationSounds|GetOnPlayerApproachScreenSound))
  (bind _onplayerleavescreensound (Variables|StationSounds|GetOnPlayerLeaveScreenSound))
  ;
  (if (ScalableFloat|GetValueasBool (Variables|StationSounds|GetTriggerAudioOnPlayerApproach))
    (if true
      (Utilities|IsValid _onplayerapproachscreensound
        (:Is%20Valid
          (Audio|SpawnSoundatLocation _onplayerapproachscreensound (Transformation|GetWorldLocation (Variables|Default|GetInteractAudioComponent)))))
      (else
        (Utilities|IsValid _onplayerleavescreensound
          (:Is%20Valid
            (Audio|SpawnSoundatLocation _onplayerleavescreensound (Transformation|GetWorldLocation (Variables|Default|GetInteractAudioComponent)))))))))

(fn ConstructionScript
  (|CallParentFunction))
